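# Translucent, unlit, two-sided glow light material for the LethalVae pickaxe effects, flagged for use on Niagara mesh particles.
# The particle color's RGB feeds the Emissive input of the shared MF_FX_EmissiveAdjustments function, so the particle system supplies the glow color.
# The graph is minimal, with no textures or local parameters, leaving any emissive adjustment to that function.
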Material /BRCosmetics/Weapons/FORT_Melee/Pickaxe_LethalVae/FX/M_Pickaxe_LethalVae_GlowLight
graph {
 properties {
  bUsedWithNiagaraMeshParticles=true
  blendMode=BLEND_Translucent
  pixelDepthOffsetMode=PDOM_Legacy
  shadingModel=MSM_Unlit
  twoSided=true
  usedShadingModels=Unlit
 }
 nodes [MaterialFunctionCall(Emissive=ParticleColor().RGB) {materialFunction=&MF_FX_EmissiveAdjustments}]
}
