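# Animation Blueprint for the LethalVae pickaxe that plays a staged procedural blade extension when the weapon is equipped.
# Each update reads WeaponCoreAnim from the owner's FortPlayerAnimInstance to set isMelee, which drives three alphas through resettable DoOnce gates and delays.
# Chained Modify Bone passes, linked by cached poses, retract and squash the blade bones instantly, then counter that in two interpolated steps, the last delayed.
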AnimBlueprint /BRCosmetics/Weapons/FORT_Melee/Pickaxe_LethalVae/Meshes/LethalVae_Axe_AnimBP
target_skeleton=&/LethalVae_Axe_Skeleton
graphs [
 AnimGraph {nodes [
   n1 = Get equipAlpha_2() -> (equipAlpha_2:double)
   n2 = "Transform (Modify) Bone - Bone: dyn_top_blade_01"(
    ComponentPose=Local To Component(LocalPose=Use cached pose 'CacheAlpha1'())
    Alpha=@n1
    Translation="15,0,0"
    Rotation="0,0,0"
    Scale="1,1,10"
   ) {
    scaleMode=BMM_Additive
    translationMode=BMM_Additive
    translationSpace=BCS_BoneSpace
    alphaBoolBlend {blendOption=Linear}
    alphaScaleBiasClamp {bInterpResult=true}
    boneToModify {boneName=dyn_top_blade_01}
    scale {z=10}
    translation {x=15}
   }
   n3 = "Transform (Modify) Bone - Bone: dyn_top_blade_02"(ComponentPose=@n2, Alpha=@n1, Translation="15,0,0", Rotation="0,0,0", Scale="1,1,1") {
    scaleMode=BMM_Replace
    translationMode=BMM_Additive
    translationSpace=BCS_BoneSpace
    alphaBoolBlend {blendOption=Linear}
    alphaScaleBiasClamp {bInterpResult=true}
    boneToModify {boneName=dyn_top_blade_02}
    translation {x=15}
   }
   n4 = "Transform (Modify) Bone - Bone: dyn_top_blade_03"(ComponentPose=@n3, Alpha=@n1, Translation="15,0,0", Rotation="0,0,0", Scale="1,1,1") {
    scaleMode=BMM_Additive
    translationMode=BMM_Additive
    translationSpace=BCS_BoneSpace
    alphaBoolBlend {blendOption=Linear}
    alphaScaleBiasClamp {bInterpResult=true}
    boneToModify {boneName=dyn_top_blade_03}
    translation {x=15}
   }
   n5 = "Transform (Modify) Bone - Bone: dyn_bottom_blade_01"(ComponentPose=@n4, Alpha=@n1, Translation="6,0,0", Rotation="0,0,0", Scale="1,1,3.3") {
    scaleMode=BMM_Additive
    translationMode=BMM_Additive
    translationSpace=BCS_BoneSpace
    alphaBoolBlend {blendOption=Linear}
    alphaScaleBiasClamp {bInterpResult=true}
    boneToModify {boneName=dyn_bottom_blade_01}
    scale {z=3.3}
    translation {x=6}
   }
   n6 = "Transform (Modify) Bone - Bone: dyn_bottom_blade_02"(ComponentPose=@n5, Alpha=@n1, Translation="6,0,0", Rotation="0,0,0", Scale="1,1,1") {
    scaleMode=BMM_Additive
    translationMode=BMM_Additive
    translationSpace=BCS_BoneSpace
    alphaBoolBlend {blendOption=Linear}
    alphaScaleBiasClamp {bInterpResult=true}
    boneToModify {boneName=dyn_bottom_blade_02}
    translation {x=6}
   }
   Output Pose(Result=Component To Local(ComponentPose=@n6)) {layerGroup=None}
   n7 = Get equipAlpha_1() -> (equipAlpha_1:double)
   n8 = "Transform (Modify) Bone - Bone: dyn_top_blade_01"(ComponentPose=Local To Component(LocalPose=Use cached pose 'CacheAlpha'()), Alpha=@n7, Translation="10,0,0", Rotation="0,0,0", Scale="1,1,1") {
    scaleMode=BMM_Additive
    translationMode=BMM_Additive
    translationSpace=BCS_BoneSpace
    alphaBoolBlend {blendOption=Linear}
    alphaScaleBiasClamp {bInterpResult=true, interpSpeedDecreasing=3, interpSpeedIncreasing=3}
    boneToModify {boneName=dyn_top_blade_01}
    translation {x=10}
   }
   n9 = "Transform (Modify) Bone - Bone: dyn_top_blade_02"(ComponentPose=@n8, Alpha=@n7, Translation="10,0,0", Rotation="0,0,0", Scale="1,1,1") {
    scaleMode=BMM_Replace
    translationMode=BMM_Additive
    translationSpace=BCS_BoneSpace
    alphaBoolBlend {blendOption=Linear}
    alphaScaleBiasClamp {bInterpResult=true, interpSpeedDecreasing=3, interpSpeedIncreasing=3}
    boneToModify {boneName=dyn_top_blade_02}
    translation {x=10}
   }
   n10 = "Transform (Modify) Bone - Bone: dyn_top_blade_03"(ComponentPose=@n9, Alpha=@n7, Translation="10,0,0", Rotation="0,0,0", Scale="1,1,1") {
    scaleMode=BMM_Additive
    translationMode=BMM_Additive
    translationSpace=BCS_BoneSpace
    alphaBoolBlend {blendOption=Linear}
    alphaScaleBiasClamp {bInterpResult=true, interpSpeedDecreasing=3, interpSpeedIncreasing=3}
    boneToModify {boneName=dyn_top_blade_03}
    translation {x=10}
   }
   n11 = "Transform (Modify) Bone - Bone: dyn_bottom_blade_01"(ComponentPose=@n10, Alpha=@n7, Translation="6,0,0", Rotation="0,0,0", Scale="1,1,3") {
    scaleMode=BMM_Additive
    translationMode=BMM_Additive
    translationSpace=BCS_BoneSpace
    alphaBoolBlend {blendOption=Linear}
    alphaScaleBiasClamp {bInterpResult=true, interpSpeedDecreasing=3, interpSpeedIncreasing=3}
    boneToModify {boneName=dyn_bottom_blade_01}
    scale {z=3}
    translation {x=6}
   }
   n12 = "Transform (Modify) Bone - Bone: dyn_bottom_blade_02"(ComponentPose=@n11, Alpha=@n7, Translation="6,0,0", Rotation="0,0,0", Scale="1,1,1") {
    scaleMode=BMM_Additive
    translationMode=BMM_Additive
    translationSpace=BCS_BoneSpace
    alphaBoolBlend {blendOption=Linear}
    alphaScaleBiasClamp {bInterpResult=true, interpSpeedDecreasing=3, interpSpeedIncreasing=3}
    boneToModify {boneName=dyn_bottom_blade_02}
    translation {x=6}
   }
   Save cached pose 'CacheAlpha1'(Pose=Component To Local(ComponentPose=@n12)) {cacheName=CacheAlpha1}
   AnimGraphNode_SubInput_0 = Input Pose()
   n13 = Get equipAlpha() -> (equipAlpha:double)
   n14 = "Transform (Modify) Bone - Bone: dyn_top_blade_01"(ComponentPose=Local To Component(LocalPose=@AnimGraphNode_SubInput_0), Alpha=@n13, Translation="-25,0,0", Rotation="0,0,0", Scale="1,1,0.1") {
    scaleMode=BMM_Additive
    translationMode=BMM_Additive
    translationSpace=BCS_BoneSpace
    alphaBoolBlend {blendOption=Linear}
    alphaScaleBiasClamp {interpSpeedDecreasing=0, interpSpeedIncreasing=0}
    boneToModify {boneName=dyn_top_blade_01}
    scale {z=0.1}
    translation {x=-25}
   }
   n15 = "Transform (Modify) Bone - Bone: dyn_top_blade_02"(ComponentPose=@n14, Alpha=@n13, Translation="-25,0,0", Rotation="0,0,0", Scale="1,1,0.1") {
    scaleMode=BMM_Replace
    translationMode=BMM_Additive
    translationSpace=BCS_BoneSpace
    alphaBoolBlend {blendOption=Linear}
    alphaScaleBiasClamp {interpSpeedDecreasing=0, interpSpeedIncreasing=0}
    boneToModify {boneName=dyn_top_blade_02}
    scale {z=0.1}
    translation {x=-25}
   }
   n16 = "Transform (Modify) Bone - Bone: dyn_top_blade_03"(ComponentPose=@n15, Alpha=@n13, Translation="-25,0,0", Rotation="0,0,0", Scale="1,1,1") {
    scaleMode=BMM_Additive
    translationMode=BMM_Additive
    translationSpace=BCS_BoneSpace
    alphaBoolBlend {blendOption=Linear}
    alphaScaleBiasClamp {interpSpeedDecreasing=0, interpSpeedIncreasing=0}
    boneToModify {boneName=dyn_top_blade_03}
    translation {x=-25}
   }
   n17 = "Transform (Modify) Bone - Bone: dyn_bottom_blade_01"(ComponentPose=@n16, Alpha=@n13, Translation="-12,0,0", Rotation="0,0,0", Scale="1,1,0.1") {
    scaleMode=BMM_Additive
    translationMode=BMM_Additive
    translationSpace=BCS_BoneSpace
    alphaBoolBlend {blendOption=Linear}
    alphaScaleBiasClamp {interpSpeedDecreasing=0, interpSpeedIncreasing=0}
    boneToModify {boneName=dyn_bottom_blade_01}
    scale {z=0.1}
    translation {x=-12}
   }
   n18 = "Transform (Modify) Bone - Bone: dyn_bottom_blade_02"(ComponentPose=@n17, Alpha=@n13, Translation="-12,0,0", Rotation="0,0,0", Scale="1,1,1") {
    scaleMode=BMM_Additive
    translationMode=BMM_Additive
    translationSpace=BCS_BoneSpace
    alphaBoolBlend {blendOption=Linear}
    alphaScaleBiasClamp {interpSpeedDecreasing=0, interpSpeedIncreasing=0}
    boneToModify {boneName=dyn_bottom_blade_02}
    translation {x=-12}
   }
   Save cached pose 'CacheAlpha'(Pose=Component To Local(ComponentPose=@n18)) {cacheName=CacheAlpha}
  ]}
 EventGraph {nodes [
   Event BlueprintUpdateAnimation()
   > n1 = Sequence!(execute) -> (then_0, then_1, then_2:Exec, then_3:Exec)
   n2 = Branch(execute=@n1.then_1, Condition=Get isMelee() -> (isMelee:bool))
   > Do Once(Reset=@n3.then)
   > Set equipAlpha(equipAlpha:double=1) -> (Output_Get:double)
   n3 = Set equipAlpha(execute=@n2.else, equipAlpha=0)
   n4 = Branch(execute=@n1.then_2, Condition=Get isMelee())
   > Do Once(Reset=@n5.then)
   > Set equipAlpha_1(equipAlpha_1:double=0) -> (Output_Get:double)
   > Delay(Duration=0)
   > Set equipAlpha_1(equipAlpha_1=1)
   n5 = Set equipAlpha_1(execute=@n4.else, equipAlpha_1=0)
   n6 = Branch(execute=@n1.then_3, Condition=Get isMelee())
   > Do Once(Reset=@n7.then)
   > Set equipAlpha_2(equipAlpha_2:double=0) -> (Output_Get:double)
   > Delay(Duration=0.8)
   > Set equipAlpha_2(equipAlpha_2=1)
   n7 = Set equipAlpha_2(execute=@n6.else, equipAlpha_2=0)
   n8 = Cast To FortPlayerAnimInstance(execute=@n1.then_0, Object=GetAnimInstance(self=GetOwnerSkeletalMeshComponent()))
   > n9 = Switch on EFortWeaponCoreAnimation!(execute, Selection:EFortWeaponCoreAnimation Enum=Get WeaponCoreAnim(self=@n8)) -> (
    Melee:Exec
    Pistol:Exec
    Shotgun:Exec
    PaperBlueprint:Exec
    Rifle:Exec
    MeleeOneHand:Exec
    MachinePistol:Exec
    RocketLauncher:Exec
    GrenadeLauncher:Exec
    GoingCommando:Exec
    AssaultRifle:Exec
    TacticalShotgun:Exec
    SniperRifle:Exec
    TrapPlacement:Exec
    ShoulderLauncher:Exec
    AbilityDecoTool:Exec
    Crossbow:Exec
    C4:Exec
    RemoteControl:Exec
    DualWield:Exec
    AR_BullPup:Exec
    AR_ForwardGrip:Exec
    MedPackPaddles:Exec
    SMG_P90:Exec
    AR_DrumGun:Exec
    Consumable_Small:Exec
    Consumable_Large:Exec
    Balloon:Exec
    MountedTurret:Exec
    CreativeTool:Exec
    ExplosiveBow:Exec
    AshtonIndigo:Exec
    AshtonChicago:Exec
    MeleeDualWield:Exec
    Unarmed:Exec
   )
   Set isMelee(
    execute=[
     @n9.AR_BullPup
     @n9.AR_DrumGun
     @n9.AR_ForwardGrip
     @n9.AbilityDecoTool
     @n9.AshtonChicago
     @n9.AshtonIndigo
     @n9.AssaultRifle
     @n9.Balloon
     @n9.C4
     @n9.Consumable_Large
     @n9.Consumable_Small
     @n9.CreativeTool
     @n9.Crossbow
     @n9.DualWield
     @n9.ExplosiveBow
     @n9.GoingCommando
     @n9.GrenadeLauncher
     @n9.MachinePistol
     @n9.MedPackPaddles
     @n9.MountedTurret
     @n9.PaperBlueprint
     @n9.Pistol
     @n9.RemoteControl
     @n9.Rifle
     @n9.RocketLauncher
     @n9.SMG_P90
     @n9.Shotgun
     @n9.ShoulderLauncher
     @n9.SniperRifle
     @n9.TacticalShotgun
     @n9.TrapPlacement
    ]
    isMelee:bool=false
   ) -> (Output_Get:bool)
   Set isMelee(execute=[@n9.Melee, @n9.MeleeDualWield, @n9.MeleeOneHand], isMelee=true)
  ]}
]
variables [
 equipAlpha {type=number}
 equipAlpha_1 {type=number}
 equipAlpha_2 {type=number}
 isMelee {type=boolean}
]
